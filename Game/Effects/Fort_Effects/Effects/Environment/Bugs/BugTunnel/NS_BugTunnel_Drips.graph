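# Niagara system that rains drip sprites from surrounding surfaces for the bug tunnel environment effect.
# The active GPU emitter spawns particles in an offset ellipsoid, snaps them to distance-field surfaces, then the depth buffer, culling invalid samples and filtering by GBuffer color hue.
# Drips fall under gravity, stretch with speed and fade in and out.
# A disabled twin emitter uses most User parameters; the active one hardcodes all but gravity.
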NiagaraSystem /Game/Effects/Fort_Effects/Effects/Environment/Bugs/BugTunnel/NS_BugTunnel_Drips
dependencies {
 usedDataInterfaces [&NiagaraDataInterfaceCollisionQuery, &NiagaraDataInterfaceCurve, &NiagaraDataInterfaceGBuffer]
 usedDynamicInputs [&Lerp_Float]
 usedModules [
  &SystemState
  &self:SystemSpawnScript.SystemScriptSource.SystemScriptGraph.NiagaraNodeAssignment_1.SetVariables_#1
 ]
}
emitters {
 RainFromObjects {
  bEnabled=false
  simTarget=GPUComputeSim
  emitterUpdateScript [
   &EmitterState()
   &SpawnRate(SpawnRate=&Multiply_Float(A=800, B=$User.Spawn Rate Mult), Use Spawn Probability=false)
  ]
  particleSpawnScript [
   &InitializeParticle(
    Lifetime Mode=::NewEnumerator1
    Sprite Rotation Mode=::NewEnumerator1
    Sprite Size Max=<50 20>
    Sprite Size Min=<25 15>
    Sprite Size Mode=::NewEnumerator2
    UsePositionOffset=false
    Write Material Random=true
   ) {module=InitializeParticle001}
   &AddVelocity(Velocity=&RandomRangeVector(Minimum=<0 0 -50>, Maximum=<0 0 -50>))
   &SphereLocation(
    MaskSpawnGroup=false
    Non Uniform Scale=<1 1 0.7>
    Offset=<0 0 1500>
    Override Seed=false
    Sphere Radius=$User.Radius Scale
    Surface Only=false
    Surface Only Band Thickness=0.5
    UseNonUniformScale=true
    UseOffset=true
    UseSphereOrientation=false
   )
   /Niagara/Modules/Math/MoveToNearestDistanceFieldSurface_GPU(Kill by Distance=800, bKillByDistance=true)
   /SetVariables(Particles.InitialPos:NiagaraPosition=$Particles.Position) {isSetParameters=true, module=SetVariables_#2}
   /Niagara/Modules/Math/PlaceParticlesOnDepthBuffer_GPU(Max Z Depth=1500, Module.Use Max Z Depth=true)
   &KillParticles(Kill Particles=&SetBoolByFloatComparison(
     A=&VectorLength(Vector=&Subtract_Vector(A=$Particles.InitialPos, B=Linked() {linkedVariable {name=Particles.Position, type=NiagaraPosition}}))
     B=1500
    ))
   &KillParticles(Kill Particles=&Boolean_NotOperation(Bool=$Output.PlaceParticlesOnDepthBuffer_GPU.ValidSample)) {module=KillParticles001}
   &SubUVAnimation(End Frame=3, SubUV Animation Mode:ENiagara_SubUVLookupMode=::NewEnumerator2)
   &CameraOffset(Camera Offset Amount=-10)
   &self:ScratchModule_02(Angle=180, BlendDirection=1, Normal:Vector3f=<0 0 1>, PickCeilingOrFloor=true, Velocity=-200) {enabled=false, module=ScratchModule_02}
  ]
  particleUpdateScript [
   &ParticleState()
   &RegionMask(Inner Float=0, Outer Float=2, Region Falloff=0, Region Radius:NiagaraFloat=$User.Rain_MaskInside_Radius, UseOffset=false, UseRotation=false) {module=RegionMask001}
   &ScaleSpriteSize(Uniform Curve Sprite Scale=DataInterface() {propertyValues {Curve {keys [{tangentMode=RCTM_User, time=0, value=0.85}, {tangentMode=RCTM_User, time=1, value=1}]}}})
   &GravityForce(Gravity=$User.Gravity Scale)
   &SolveForcesAndVelocity(Clamp Velocity=false, Limit Acceleration=false, Speed Limit=500)
   &ScaleSpriteSizeBySpeed(
    Max Scale Factor=<0.1 5>
    Min Scale Factor=<0.05 1>
    Source Velocity:Vector3f=&Multiply_Vector(A=<0.05 0.05 1>, B=$Particles.Velocity)
   )
   &ScaleColor(
    Scale Alpha=&Multiply_Float(
     A=&Multiply_Float(
      A=&Multiply_Float(
       A=&FloatFromCurve(FloatCurve=DataInterface() {propertyValues {Curve {keys [
            {interpMode=RCIM_Cubic, tangentMode=RCTM_User, time=0, value=0}
            {interpMode=RCIM_Cubic, tangentMode=RCTM_User, time=0.5, value=1}
            {tangentMode=RCTM_User, time=1, value=0}
           ]}}})
       B=1
      )
      B=1
     )
     B=$User.Rain_AlphaMult
    )
    Scale RGB=&MakeVectorFromLinearColorRGB(Color=&MultiplyLinearColor(Linear Color=$User.Rain_NonLitColor, Scale Factor=&MakeVector4_FromFloat(Float=$User.Rain_ColorMult)))
    ScaleA=true
    ScaleRGB=true
   )
  ]
  rendererClasses [&NiagaraSpriteRendererProperties]
  renderers [
   {rendererClass=&NiagaraSpriteRendererProperties}
   {rendererClass=&NiagaraSpriteRendererProperties, rendererIndex=1}
  ]
 }
 RainFromObjects001 {
  simTarget=GPUComputeSim
  emitterUpdateScript [
   &EmitterState()
   &SpawnRate(SpawnRate=800, Use Spawn Probability=false)
  ]
  particleSpawnScript [
   &InitializeParticle(
    Color Mode=::NewEnumerator0
    Lifetime Mode=::NewEnumerator1
    Sprite Rotation Mode=::NewEnumerator1
    Sprite Size Max=<50 20>
    Sprite Size Min=<25 15>
    Sprite Size Mode=::NewEnumerator2
    UsePositionOffset=false
    Write Material Random=true
   ) {module=InitializeParticle001}
   &AddVelocity(Velocity=&RandomRangeVector(Minimum=<0 0 -50>, Maximum=<0 0 -50>))
   &SphereLocation(
    MaskSpawnGroup=false
    Non Uniform Scale=<1 2 0.7>
    Offset=<0 0 1500>
    Override Seed=false
    Sphere Radius=2500
    Surface Only=false
    Surface Only Band Thickness=0.5
    UseNonUniformScale=true
    UseOffset=true
    UseSphereOrientation=false
   )
   &MoveToNearestDistanceFieldSurface_GPU(Kill by Distance=800, bKillByDistance=true)
   /SetVariables(Particles.InitialPos:NiagaraPosition=$Particles.Position) {isSetParameters=true, module=SetVariables_#2}
   &PlaceParticlesOnDepthBuffer_GPU(Max Z Depth=1500, Module.Use Max Z Depth=true)
   &KillParticles(Kill Particles=&SetBoolByFloatComparison(
     A=&VectorLength(Vector=&Subtract_Vector(A=$Particles.InitialPos, B=Linked() {linkedVariable {name=Particles.Position, type=NiagaraPosition}}))
     B=1500
    ))
   &KillParticles(Kill Particles=&Boolean_NotOperation(Bool=$Output.PlaceParticlesOnDepthBuffer_GPU.ValidSample)) {module=KillParticles001}
   &SubUVAnimation(End Frame=3, SubUV Animation Mode:ENiagara_SubUVLookupMode=::NewEnumerator2) {module=SubUVAnimation001}
   &CameraOffset(Camera Offset Amount=-10)
   &self:ScratchModule_02(Angle=180, BlendDirection=1, Normal:Vector3f=<0 0 1>, PickCeilingOrFloor=true, Velocity=-200) {enabled=false, module=ScratchModule_02}
   /Game/Effects/Niagara/Modules/Spawn/FN_Spawn_GBuffer_Color_WithinRangeV2(HueMax=32, HueMin=26)
   &KillParticlesInVolume(
    Box Size:Vector3f=<1600 11000 900>
    Invert Volume=true
    Kill Shape=::NewEnumerator1
    Volume Origin:Vector3f=&AddVectorToPosition(Position=<-18788 77936 5704>, Vector=<0 0 0>)
   ) {enabled=false}
   &KillParticlesInVolume(
    Box Size:Vector3f=<11000 1600 900>
    Invert Volume=true
    Kill Shape=::NewEnumerator1
    Volume Origin:Vector3f=&AddVectorToPosition(Position=<-29140 66772 5768>, Vector=<0 0 0>)
   ) {enabled=false, module=KillParticlesInVolume001}
  ]
  particleUpdateScript [
   &ParticleState()
   &RegionMask(Inner Float=0, Outer Float=2, Region Falloff=0, Region Radius=500, UseOffset=false, UseRotation=false) {enabled=false, module=RegionMask001}
   &ScaleSpriteSize(Uniform Curve Sprite Scale=DataInterface() {propertyValues {Curve {keys [{tangentMode=RCTM_User, time=0, value=0.85}, {tangentMode=RCTM_User, time=1, value=1}]}}})
   &GravityForce(Gravity=$User.Gravity Scale)
   &SolveForcesAndVelocity(Clamp Velocity=false, Limit Acceleration=false, Speed Limit=500)
   &ScaleSpriteSizeBySpeed(
    Max Scale Factor=<0.1 5>
    Min Scale Factor=<0.05 1>
    Source Velocity:Vector3f=&Multiply_Vector(A=<0.05 0.05 1>, B=$Particles.Velocity)
   )
   &ScaleColor(
    Scale Alpha=&FloatFromCurve(FloatCurve=DataInterface() {propertyValues {Curve {keys [
         {interpMode=RCIM_Cubic, tangentMode=RCTM_User, time=0, value=0}
         {interpMode=RCIM_Cubic, tangentMode=RCTM_User, time=0.5, value=1}
         {tangentMode=RCTM_User, time=1, value=0}
        ]}}})
    Scale RGB=&MakeVectorFromLinearColorRGB(Color=&MultiplyLinearColor(Linear Color=<rgba 0.4492 0.547 0.75 1>, Scale Factor=&MakeVector4_FromFloat(Float=1)))
    ScaleA=true
    ScaleRGB=true
   )
  ]
  rendererClasses [&NiagaraSpriteRendererProperties]
  renderers [
   {rendererClass=&NiagaraSpriteRendererProperties}
   {rendererClass=&NiagaraSpriteRendererProperties, rendererIndex=1}
  ]
 }
}
summary {
 systemName=NS_BugTunnel_Drips
 emitters [
  {bEnabled=false, emitterName=RainFromObjects, simTarget=GPUComputeSim, rendererClasses [&NiagaraSpriteRendererProperties]}
  {emitterName=RainFromObjects001, simTarget=GPUComputeSim, rendererClasses [&NiagaraSpriteRendererProperties]}
 ]
 userVariables [
  User.Gravity Scale {type=Vector3f, description="", defaultValue {struct=Vector3f, value=<0 0 -400>}}
  User.Radius Scale {type=NiagaraFloat, defaultValue=2500, description=""}
  User.Rain_AlphaMult {type=NiagaraFloat, defaultValue=1, description=""}
  User.Rain_ColorMult {type=NiagaraFloat, defaultValue=1, description=""}
  User.Rain_FallingSpeed {type=NiagaraFloat, defaultValue=1.5, description=""}
  User.Rain_InheritVel {type=Vector3f, description="", defaultValue {struct=Vector3f, value=<-1 -1 0>}}
  User.Rain_MaskInside_Radius {type=NiagaraFloat, defaultValue=500, description=""}
  User.Rain_NonLitColor {type=LinearColor, description="", defaultValue {struct=LinearColor, value=<rgba 0.4492 0.547 0.75 1>}}
  User.Rain_SpawnRateScale {type=NiagaraFloat, defaultValue=2, description=""}
  User.Spawn Rate Mult {type=NiagaraFloat, defaultValue=1, description=""}
 ]
}
