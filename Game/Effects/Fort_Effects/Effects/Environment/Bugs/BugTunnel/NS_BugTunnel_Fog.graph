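# Niagara system for drifting, faint blue-tinted fog in the BugTunnel environment.
# A single Fog emitter continuously spawns randomly sized particles in a wide, shallow cylinder, drawn by two sprite renderers and one mesh renderer.
# Alpha fades in and out over lifetime, while acceleration along X with drag creates the drift.
# An LOD distance remap shrinks sprites and meshes at close range, and a camera offset shifts the particles.
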NiagaraSystem /Game/Effects/Fort_Effects/Effects/Environment/Bugs/BugTunnel/NS_BugTunnel_Fog
dependencies {usedDataInterfaces [&NiagaraDataInterfaceCurve], usedDynamicInputs [], usedModules [&SystemState]}
emitters {Fog {
  emitterUpdateScript [
   &EmitterState()
   &SpawnRate(SpawnRate=5, Use Spawn Probability=false)
  ]
  particleSpawnScript [
   &InitializeParticle(
    Color=<rgba 0.1 0.1896 0.2908 1>
    Lifetime=&RandomRangeFloat(Minimum=4, Maximum=6)
    Mesh Scale Mode=::NewEnumerator1
    Mesh Uniform Scale Max=12
    Mesh Uniform Scale Min=6
    Sprite Size Mode=::NewEnumerator1
    Uniform Sprite Size Max=750
    Uniform Sprite Size Min=350
    UsePositionOffset=false
    Write Material Random=true
   )
   &ShapeLocation(
    Cylinder Height=5
    Cylinder Radius=3500
    MaskSpawnGroup=false
    Override Seed=false
    Shape Primitive=::NewEnumerator1
    Surface Only=false
    UseOwnerScale=true
   )
  ]
  particleUpdateScript [
   &ParticleState()
   &ScaleMeshSize(Scale Factor=&Lerp_Vector(
     A=<0.125 0.125 0.125>
     B=<1 1 1>
     Alpha=&FN_LOD_Distance_Remap(Start Distance=0, End Distance=3500, Minimum Output=0, Maximum Output=1, Clamp Output=true)
    ))
   &ScaleSpriteSize(
    Uniform Curve Index=&FN_LOD_Distance_Remap(Start Distance=0, End Distance=1500, Minimum Output=0, Maximum Output=1, Clamp Output=true)
    Uniform Curve Sprite Scale=DataInterface() {propertyValues {Curve {keys [{interpMode=RCIM_Cubic, tangentMode=RCTM_Break, time=0, value=0}, {tangentMode=RCTM_User, time=1, value=1}]}}}
   )
   &ScaleColor(
    Scale Alpha=&FloatFromCurve(
     FloatCurve=DataInterface() {propertyValues {Curve {keys [
         {interpMode=RCIM_Cubic, tangentMode=RCTM_User, time=0, value=0}
         {interpMode=RCIM_Cubic, tangentMode=RCTM_User, time=0.5, value=1}
         {tangentMode=RCTM_User, time=1, value=0}
        ]}}}
     Scale Curve=0.1
    )
    ScaleA=true
    ScaleRGB=true
   ) {module=ScaleColor001}
   &AccelerationForce(Acceleration=&RandomRangeVector(Minimum=<80 0 0>, Maximum=<135 1 1>))
   &Drag(Use Linear Drag=true, Use Rotational Drag=true)
   &SolveForcesAndVelocity(Clamp Velocity=false, Limit Acceleration=false)
   &CameraOffset(Camera Offset Amount=99)
  ]
  rendererClasses [&NiagaraSpriteRendererProperties, &NiagaraMeshRendererProperties]
  renderers [
   {rendererClass=&NiagaraSpriteRendererProperties}
   {rendererClass=&NiagaraMeshRendererProperties, rendererIndex=1}
   {rendererClass=&NiagaraSpriteRendererProperties, rendererIndex=2}
  ]
 }}
summary {
 systemName=NS_BugTunnel_Fog
 emitters [{emitterName=Fog, rendererClasses [&NiagaraSpriteRendererProperties, &NiagaraMeshRendererProperties]}]
 userVariables []
}
