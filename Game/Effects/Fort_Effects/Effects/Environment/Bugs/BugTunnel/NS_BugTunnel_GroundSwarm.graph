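# Niagara system for a swarm of mesh bugs crawling over bug tunnel surfaces; only Bugs002 of three GPU emitters is enabled.
# Bugs spawn in a flattened, elongated volume, snap to distance fields and orient to surface normal and velocity.
# A crawling module, neighbor grid, spring to spawn, camera-proximity force and distance-gated curl noise drive motion.
# Bugs at the NPC-supplied player position are killed, with a death switch using a system position array.
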NiagaraSystem /Game/Effects/Fort_Effects/Effects/Environment/Bugs/BugTunnel/NS_BugTunnel_GroundSwarm
dependencies {
 usedDataInterfaces [
  &NiagaraDataInterfaceParticleRead
  &NiagaraDataInterfaceCollisionQuery
  &NiagaraDataInterfaceCamera
  &NiagaraDataInterfaceDebugDraw
  &NiagaraDataInterfaceGBuffer
  &NiagaraDataInterfaceArrayPosition
 ]
 usedDynamicInputs []
 usedModules [
  &self:SystemSpawnScript.SystemScriptSource.SystemScriptGraph.NiagaraNodeAssignment_3.SetVariables_#1
  &SystemState
  &self:ReadDeath
  &self:SystemSpawnScript.SystemScriptSource.SystemScriptGraph.NiagaraNodeAssignment_0.SetVariables_#2
  &self:SystemSpawnScript.SystemScriptSource.SystemScriptGraph.NiagaraNodeAssignment_4.SetVariables_#3
 ]
}
emitters {
 Bugs {
  bEnabled=false
  simTarget=GPUComputeSim
  emitterSpawnScript [
   /Niagara/Modules/Collision/InitializeNeighborGrid(
    Grid Extents:Vector3f=<7500 7500 7500>
    Local Pivot:Vector3f=<0.5 0.5 0.125>
    MaxNeighborsPerCell=64
    NumCellsX=8
    NumCellsY=8
    NumCellsZ=8
    Offset:NiagaraPosition=$Engine.Owner.Position
   )
   /SetVariables(Emitter.BoidReader=DataInterface() {propertyValues {EmitterBinding {bindingMode=Self, emitterName=None}}}) {isSetParameters=true, module=SetVariables_#4}
  ]
  emitterUpdateScript [
   &EmitterState(Life Cycle Mode=::NewEnumerator1, Loop Behavior=::NewEnumerator1, Loop Duration Mode=::NewEnumerator1, UseLoopDelay=false)
   &SpawnBurst_Instantaneous(Spawn Count=$User.SpawnCount, Use Spawn Probability=false, UseLoopCountLimit=false)
   /SetVariables(Emitter.Enable Boids:NiagaraBool=&SetBoolByFloatComparison(
     A=&VectorLength(Vector=&SubtractVectorFromPosition(
       Position=&ReturnCameraPosition(Coordinate Space:ENiagaraCoordinateSpace=::Simulation, Camera Query:NiagaraDataInterfaceCamera=$System.Camera Query)
       Vector=/Niagara/DynamicInputs/Helpers/SimulationPosition()
      ))
     B=2500
     Comparison Type=::NewEnumerator6
    )) {isSetParameters=true, module=SetVariables_#5}
  ]
  particleSpawnScript [
   &InitializeParticle(
    Color=<rgba 0 0 0 1>
    Lifetime=5
    Mesh Scale Mode=::NewEnumerator1
    Mesh Uniform Scale Max=0.075
    Mesh Uniform Scale Min=0.065
    UsePositionOffset=false
    Write Material Random=true
   )
   &ShapeLocation(
    MaskSpawnGroup=false
    Non Uniform Scale=$User.RadiusShape
    Offset=<0 0 500>
    Offset Mode=::NewEnumerator0
    Override Seed=false
    Sphere Radius=&Multiply_Float(A=1500, B=$User.RadiusMult)
    UseOwnerScale=true
   )
   &self:Measure() {module=Measure}
   /Niagara/Modules/Math/MoveToNearestDistanceFieldSurface_GPU(Kill by Distance=2500, Radius=6, bKillByDistance=false)
   &InitialMeshOrientation(
    Mesh Axis To Orient=<0 0 -1>
    Mesh Orientation Mode=::NewEnumerator2
    Orientation Vector=$Output.MoveToNearestDistanceFieldSurface_GPU.VectorToNearestSurfaceNormal
    Use Rotation Vector=false
   )
   /SetVariables(Particles.Awake=true, Particles.BirdMeshScale:Vector3f=$Particles.Scale) {isSetParameters=true, module=SetVariables_#6}
  ]
  particleUpdateScript [
   &ParticleState(Kill Particles When Lifetime Has Elapsed=false, Let Infinitely Lived Particles Die When Emitter Deactivates=true)
   /SetVariables(Particles.Player Close:NiagaraBool=&SetBoolByFloatComparison(
     A=&VectorLength(Vector=&SubtractVectorFromPosition(
       Position=&ReturnCameraPosition(Coordinate Space:ENiagaraCoordinateSpace=::Simulation, Camera Query:NiagaraDataInterfaceCamera=$System.Camera Query)
       Vector=Linked() {linkedVariable {name=Particles.Position, type=NiagaraPosition}}
      ))
     B=500
     Comparison Type=::NewEnumerator6
    )) {isSetParameters=true, module=SetVariables_#7}
   &MoveToNearestDistanceFieldSurface_GPU(Kill by Distance=8000, Radius=500, bKillByDistance=true) {module=MoveToNearestDistanceFieldSurface_GPU001}
   &self:BugsCrawling(Position:NiagaraPosition=$Particles.Position, Velocity:Vector3f=$Particles.Velocity) {module=BugsCrawling}
   &PointForce(
    Force Falloff Distance:NiagaraFloat=&Power(a=&RandomRangeFloat(Minimum=500, Maximum=5000), b=8)
    Force Origin=&AddVectorToPosition(
     Position=&ReturnCameraPosition(Coordinate Space:ENiagaraCoordinateSpace=::World, Camera Query:NiagaraDataInterfaceCamera=DataInterface())
     Vector=<0 0 -500>
    )
    Force Strength=&MakeCustomFloatFromBool(Bool=$Particles.Player Close, True Float=30000, False Float=0)
    Module.Use Falloff=true
    Module.UseLinearForce=false
   )
   &LinearForce(
    Coordinate Space:ENiagaraCoordinateSpace=::World
    Force=&Multiply_VectorByFloat(
     Vector=&Subtract_Vector(
      A=Linked() {linkedVariable {name=Particles.Initial.Position, type=NiagaraPosition}}
      B=Linked() {linkedVariable {name=Particles.Position, type=NiagaraPosition}}
     )
     Float=&RandomRangeFloat(Minimum=0.25, Maximum=2.5)
    )
   )
   &Drag(Drag=4, Use Linear Drag=true, Use Rotational Drag=true)
   &CurlNoiseForce(
    Bias Noise=false
    Noise Frequency=5
    Noise Strength=&MakeCustomFloatFromBool(Bool=&Boolean_NotOperation(Bool=$Emitter.Enable Boids), True Float=1750, False Float=0)
    Pan Noise=true
    Pan Noise Field=<0 0 -1>
   )
   &SolveForcesAndVelocity(Acceleration Limit=350, Clamp Velocity=true, Limit Acceleration=true, Speed Limit=300)
   &self:Compute_Matrix(
    Normal:Vector3f=&Multiply_VectorByFloat(Vector=$Output.MoveToNearestDistanceFieldSurface_GPU001.VectorToNearestSurfaceNormal, Float=-1)
    Position:NiagaraPosition=$Particles.Position
    Velocity:Vector3f=$Particles.Velocity
   ) {module=Compute_Matrix}
   /SetVariables(Particles.MeshOrientation=/Niagara/DynamicInputs/Math/GetMatrixRotation(Matrix:NiagaraMatrix=$Output.Compute_Matrix.NiagaraMatrix)) {isSetParameters=true, module=SetVariables_#8}
   &InterpolateOverTime(Rate Of Change=4, Target Quat Value:Quat4f=$Particles.MeshOrientation, Variable Type=::NewEnumerator5)
   /SetVariables(Particles.MeshOrientation=&LerpQuaternion(
     Quaternion A:Quat4f=$Particles.Initial.MeshOrientation
     Quaternion B:Quat4f=$Particles.InterpolateOverTime.InterpolateOverTime_Quaternion.Moving Average
     Lerp Factor:NiagaraFloat=&RemapRange(
      Input Value To Remap=&VectorLength(Vector=$Particles.Presolve.Velocity)
      Input Min=0
      Input Max=10
      Output Min=0
      Output Max=1
     )
     Enable Spherical Linear Interpolation:NiagaraBool=true
    )) {isSetParameters=true, module=SetVariables_#9}
   /SetVariables(Particles.ParticleUpdateForce:Vector3f=$Transient.PhysicsForce) {isSetParameters=true, module=SetVariables_#10}
   /NiagaraFluids/Modules/Grid3D/NeighborGrid3D_Visualize(
    GridToVisualize:NiagaraDataInterfaceNeighborGrid3D=$Emitter.InitializeNeighborGrid.NeighborGrid3D
    UnitToWorld:NiagaraMatrix=$Emitter.GridUnitToWorld
   ) {enabled=false}
   &KillParticlesInVolume(
    Sphere Radius=80
    Volume Origin:Vector3f=&MakeVectorFromLinearColorRGB(Color=$NPC.FortniteNPC.PlayerPosition)
   )
   &RegionMask(
    Region Origin:Vector3f=&MakeVectorFromLinearColorRGB(Color=$NPC.FortniteNPC.PlayerPosition)
    Region Radius=90
    UseOffset=false
    UseRotation=false
   )
   &self:DeathSwitch_0_to_1(
    IsOutsideMask:NiagaraBool=&Boolean_NotOperation(Bool=$Particles.RegionMask.IsInsideMask)
    Position Array:NiagaraDataInterfaceArrayPosition=$System.Death Array
   ) {module=DeathSwitch_0_to_1}
   &Color(Color=&MakeCustomLinearColorFromBool(Bool=$Particles.RegionMask.IsInsideMask, True Color=<rgba 1 1 1 1>, False Color=<rgba 0 0 0 1>)) {enabled=false}
   &KillParticles(Kill Particles=&self:ScratchDynamicInput_01())
  ]
  rendererClasses [&NiagaraMeshRendererProperties]
  renderers [{rendererClass=&NiagaraMeshRendererProperties, rendererIndex=0}]
 }
 Bugs001 {
  bEnabled=false
  simTarget=GPUComputeSim
  emitterSpawnScript [
   &InitializeNeighborGrid(
    Grid Extents:Vector3f=<7500 7500 7500>
    Local Pivot:Vector3f=<0.5 0.5 0.125>
    MaxNeighborsPerCell=64
    NumCellsX=8
    NumCellsY=8
    NumCellsZ=8
    Offset:NiagaraPosition=$Engine.Owner.Position
   )
   /SetVariables(Emitter.BoidReader=DataInterface() {propertyValues {EmitterBinding {bindingMode=Self, emitterName=None}}}) {isSetParameters=true, module=SetVariables_#4}
  ]
  emitterUpdateScript [
   &EmitterState(Life Cycle Mode=::NewEnumerator1, Loop Behavior=::NewEnumerator1, Loop Duration Mode=::NewEnumerator1, UseLoopDelay=false)
   &SpawnBurst_Instantaneous(Spawn Count=$User.SpawnCount, Use Spawn Probability=false, UseLoopCountLimit=false)
   /SetVariables(Emitter.Enable Boids:NiagaraBool=&SetBoolByFloatComparison(
     A=&VectorLength(Vector=&SubtractVectorFromPosition(
       Position=&ReturnCameraPosition(Coordinate Space:ENiagaraCoordinateSpace=::Simulation, Camera Query:NiagaraDataInterfaceCamera=$System.Camera Query)
       Vector=&SimulationPosition()
      ))
     B=2500
     Comparison Type=::NewEnumerator6
    )) {isSetParameters=true, module=SetVariables_#5}
  ]
  particleSpawnScript [
   &InitializeParticle(
    Color Maximum=<rgba 0.1279 0 0.0313 1>
    Color Mode=::NewEnumerator2
    Lifetime=5
    Mesh Scale Mode=::NewEnumerator3
    Mesh Uniform Scale=&Multiply_Float(A=&RandomRangeFloat(Minimum=0.065, Maximum=0.095), B=0.7)
    UsePositionOffset=false
    Write Material Random=true
   )
   &ShapeLocation(
    MaskSpawnGroup=false
    Non Uniform Scale=<12 1 0.25>
    Offset=<0 0 -300>
    Offset Mode=::NewEnumerator0
    Override Seed=false
    Sphere Radius=700
    UseOwnerScale=true
   )
   &self:Measure() {module=Measure}
   &MoveToNearestDistanceFieldSurface_GPU(Kill by Distance=2500, Radius=6, bKillByDistance=false)
   &InitialMeshOrientation(
    Mesh Axis To Orient=<0 0 -1>
    Mesh Orientation Mode=::NewEnumerator2
    Orientation Vector=$Output.MoveToNearestDistanceFieldSurface_GPU.VectorToNearestSurfaceNormal
    Use Rotation Vector=false
   )
   /SetVariables(Particles.Awake=true, Particles.BirdMeshScale:Vector3f=$Particles.Scale) {isSetParameters=true, module=SetVariables_#6}
   /Game/Effects/Niagara/Modules/Spawn/FN_Spawn_GBuffer_Color_WithinRangeV2(HueMax=76, HueMin=45) {enabled=false}
   &Color(Color=&RandomRangeLinearColor(Minimum=<rgba 0.3005 0.3095 0.2747 1>, Maximum=<rgba 0.1683 0.162 0.1413 1>)) {enabled=false, module=Color001}
   &self:ScratchModule_02(Angle=70, BlendDirection=1, Normal:Vector3f=<0 0 1>, PickCeilingOrFloor=true, Velocity=-200) {enabled=false, module=ScratchModule_02}
   &KillParticlesInVolume(
    Box Size:Vector3f=<8000 1500 300>
    Kill Shape=::NewEnumerator1
    Volume Origin:Vector3f=&AddVectorToPosition(Position=<-31776 66792 5216>, Vector=<0 0 0>)
   ) {module=KillParticlesInVolume001}
  ]
  particleUpdateScript [
   &ParticleState(
    Kill Particles When Lifetime Has Elapsed=false
    Let Infinitely Lived Particles Die When Emitter Deactivates=true
    Loop Particles Lifetime=true
   )
   /SetVariables(Particles.Player Close:NiagaraBool=&SetBoolByFloatComparison(
     A=&VectorLength(Vector=&SubtractVectorFromPosition(
       Position=&ReturnCameraPosition(Coordinate Space:ENiagaraCoordinateSpace=::Simulation, Camera Query:NiagaraDataInterfaceCamera=$System.Camera Query)
       Vector=Linked() {linkedVariable {name=Particles.Position, type=NiagaraPosition}}
      ))
     B=500
     Comparison Type=::NewEnumerator6
    )) {isSetParameters=true, module=SetVariables_#7}
   &MoveToNearestDistanceFieldSurface_GPU(Kill by Distance=8000, Radius=500, bKillByDistance=true) {module=MoveToNearestDistanceFieldSurface_GPU001}
   &self:BugsCrawling(Position:NiagaraPosition=$Particles.Position, Velocity:Vector3f=$Particles.Velocity) {module=BugsCrawling}
   &PointForce(
    Force Falloff Distance:NiagaraFloat=&Power(a=&RandomRangeFloat(Minimum=500, Maximum=5000), b=8)
    Force Origin=&AddVectorToPosition(
     Position=&ReturnCameraPosition(Coordinate Space:ENiagaraCoordinateSpace=::World, Camera Query:NiagaraDataInterfaceCamera=DataInterface())
     Vector=<0 0 -500>
    )
    Force Strength=&MakeCustomFloatFromBool(Bool=$Particles.Player Close, True Float=30000, False Float=0)
    Module.Use Falloff=true
    Module.UseLinearForce=false
   )
   &LinearForce(
    Coordinate Space:ENiagaraCoordinateSpace=::World
    Force=&Multiply_VectorByFloat(
     Vector=&Subtract_Vector(
      A=Linked() {linkedVariable {name=Particles.Initial.Position, type=NiagaraPosition}}
      B=Linked() {linkedVariable {name=Particles.Position, type=NiagaraPosition}}
     )
     Float=&RandomRangeFloat(Minimum=0.25, Maximum=2.5)
    )
   )
   &Drag(Drag=4, Use Linear Drag=true, Use Rotational Drag=true)
   &CurlNoiseForce(
    Bias Noise=false
    Noise Frequency=5
    Noise Strength=&MakeCustomFloatFromBool(Bool=&Boolean_NotOperation(Bool=$Emitter.Enable Boids), True Float=1750, False Float=0)
    Pan Noise=true
    Pan Noise Field=<0 0 -1>
   )
   &SolveForcesAndVelocity(Acceleration Limit=350, Clamp Velocity=true, Limit Acceleration=true, Speed Limit=300)
   &self:Compute_Matrix(
    Normal:Vector3f=&Multiply_VectorByFloat(Vector=$Output.MoveToNearestDistanceFieldSurface_GPU001.VectorToNearestSurfaceNormal, Float=-1)
    Position:NiagaraPosition=$Particles.Position
    Velocity:Vector3f=$Particles.Velocity
   ) {module=Compute_Matrix}
   /SetVariables(Particles.MeshOrientation=&GetMatrixRotation(Matrix:NiagaraMatrix=$Output.Compute_Matrix.NiagaraMatrix)) {isSetParameters=true, module=SetVariables_#8}
   &InterpolateOverTime(Rate Of Change=4, Target Quat Value:Quat4f=$Particles.MeshOrientation, Variable Type=::NewEnumerator5)
   /SetVariables(Particles.MeshOrientation=&LerpQuaternion(
     Quaternion A:Quat4f=$Particles.Initial.MeshOrientation
     Quaternion B:Quat4f=$Particles.InterpolateOverTime.InterpolateOverTime_Quaternion.Moving Average
     Lerp Factor:NiagaraFloat=&RemapRange(
      Input Value To Remap=&VectorLength(Vector=$Particles.Presolve.Velocity)
      Input Min=0
      Input Max=10
      Output Min=0
      Output Max=1
     )
     Enable Spherical Linear Interpolation:NiagaraBool=true
    )) {isSetParameters=true, module=SetVariables_#9}
   /SetVariables(Particles.ParticleUpdateForce:Vector3f=$Transient.PhysicsForce) {isSetParameters=true, module=SetVariables_#10}
   &NeighborGrid3D_Visualize(
    GridToVisualize:NiagaraDataInterfaceNeighborGrid3D=$Emitter.InitializeNeighborGrid.NeighborGrid3D
    UnitToWorld:NiagaraMatrix=$Emitter.GridUnitToWorld
   ) {enabled=false}
   &KillParticlesInVolume(
    Sphere Radius=80
    Volume Origin:Vector3f=&MakeVectorFromLinearColorRGB(Color=$NPC.FortniteNPC.PlayerPosition)
   )
   &RegionMask(
    Region Origin:Vector3f=&MakeVectorFromLinearColorRGB(Color=$NPC.FortniteNPC.PlayerPosition)
    Region Radius=90
    UseOffset=false
    UseRotation=false
   )
   &self:DeathSwitch_0_to_1(
    IsOutsideMask:NiagaraBool=&Boolean_NotOperation(Bool=$Particles.RegionMask.IsInsideMask)
    Position Array:NiagaraDataInterfaceArrayPosition=$System.Death Array
   ) {module=DeathSwitch_0_to_1}
   &Color(Color=&MakeCustomLinearColorFromBool(Bool=$Particles.RegionMask.IsInsideMask, True Color=<rgba 1 1 1 1>, False Color=<rgba 0 0 0 1>)) {enabled=false}
   &KillParticles(Kill Particles=&self:ScratchDynamicInput_01())
  ]
  rendererClasses [&NiagaraMeshRendererProperties]
  renderers [{rendererClass=&NiagaraMeshRendererProperties, rendererIndex=0}]
 }
 Bugs002 {
  simTarget=GPUComputeSim
  emitterSpawnScript [
   &InitializeNeighborGrid(
    Grid Extents:Vector3f=<7500 7500 7500>
    Local Pivot:Vector3f=<0.5 0.5 0.125>
    MaxNeighborsPerCell=64
    NumCellsX=8
    NumCellsY=8
    NumCellsZ=8
    Offset:NiagaraPosition=$Engine.Owner.Position
   )
   /SetVariables(Emitter.BoidReader=DataInterface() {propertyValues {EmitterBinding {bindingMode=Self, emitterName=None}}}) {isSetParameters=true, module=SetVariables_#4}
  ]
  emitterUpdateScript [
   &EmitterState(Life Cycle Mode=::NewEnumerator1, Loop Behavior=::NewEnumerator1, Loop Duration Mode=::NewEnumerator1, UseLoopDelay=false)
   &SpawnBurst_Instantaneous(Spawn Count=$User.SpawnCount, Use Spawn Probability=false, UseLoopCountLimit=false)
   /SetVariables(Emitter.Enable Boids:NiagaraBool=&SetBoolByFloatComparison(
     A=&VectorLength(Vector=&SubtractVectorFromPosition(
       Position=&ReturnCameraPosition(Coordinate Space:ENiagaraCoordinateSpace=::Simulation, Camera Query:NiagaraDataInterfaceCamera=$System.Camera Query)
       Vector=&SimulationPosition()
      ))
     B=2500
     Comparison Type=::NewEnumerator6
    )) {isSetParameters=true, module=SetVariables_#5}
  ]
  particleSpawnScript [
   &InitializeParticle(
    Color Maximum=<rgba 0.1279 0 0.0313 1>
    Color Mode=::NewEnumerator2
    Lifetime=5
    Mesh Scale Mode=::NewEnumerator3
    Mesh Uniform Scale=&Multiply_Float(A=&RandomRangeFloat(Minimum=0.065, Maximum=0.095), B=$User.MeshScaleMult)
    UsePositionOffset=false
    Write Material Random=true
   )
   &ShapeLocation(
    MaskSpawnGroup=false
    Non Uniform Scale=<12 1 0.25>
    Offset=<0 0 -300>
    Offset Mode=::NewEnumerator0
    Override Seed=false
    Sphere Radius=&Multiply_Float(A=700, B=$User.RadiusMult)
    UseOwnerScale=true
   )
   &self:Measure() {module=Measure}
   &MoveToNearestDistanceFieldSurface_GPU(Kill by Distance=2500, Radius=6, bKillByDistance=false)
   &InitialMeshOrientation(
    Mesh Axis To Orient=<0 0 -1>
    Mesh Orientation Mode=::NewEnumerator2
    Orientation Vector=$Output.MoveToNearestDistanceFieldSurface_GPU.VectorToNearestSurfaceNormal
    Use Rotation Vector=false
   )
   /SetVariables(Particles.Awake=true, Particles.BirdMeshScale:Vector3f=$Particles.Scale) {isSetParameters=true, module=SetVariables_#6}
   &FN_Spawn_GBuffer_Color_WithinRangeV2(HueMax=40, HueMin=29) {enabled=false}
   &Color(Color=&RandomRangeLinearColor(Minimum=<rgba 0.4342 0.4179 0.3864 1>, Maximum=<rgba 0.3613 0.3372 0.314 1>)) {enabled=false, module=Color001}
   &KillParticlesInVolume(
    Box Size:Vector3f=<8000 1500 300>
    Kill Shape=::NewEnumerator1
    Volume Origin:Vector3f=&AddVectorToPosition(Position=<-31776 66792 5216>, Vector=<0 0 0>)
   ) {module=KillParticlesInVolume001}
  ]
  particleUpdateScript [
   &ParticleState(
    Kill Particles When Lifetime Has Elapsed=false
    Let Infinitely Lived Particles Die When Emitter Deactivates=true
    Loop Particles Lifetime=true
   )
   /SetVariables(Particles.Player Close:NiagaraBool=&SetBoolByFloatComparison(
     A=&VectorLength(Vector=&SubtractVectorFromPosition(
       Position=&ReturnCameraPosition(Coordinate Space:ENiagaraCoordinateSpace=::Simulation, Camera Query:NiagaraDataInterfaceCamera=$System.Camera Query)
       Vector=Linked() {linkedVariable {name=Particles.Position, type=NiagaraPosition}}
      ))
     B=500
     Comparison Type=::NewEnumerator6
    )) {isSetParameters=true, module=SetVariables_#7}
   &MoveToNearestDistanceFieldSurface_GPU(Kill by Distance=8000, Radius=500, bKillByDistance=true) {module=MoveToNearestDistanceFieldSurface_GPU001}
   &self:BugsCrawling(Position:NiagaraPosition=$Particles.Position, Velocity:Vector3f=$Particles.Velocity) {module=BugsCrawling}
   &PointForce(
    Force Falloff Distance:NiagaraFloat=&Power(a=&RandomRangeFloat(Minimum=500, Maximum=5000), b=8)
    Force Origin=&AddVectorToPosition(
     Position=&ReturnCameraPosition(Coordinate Space:ENiagaraCoordinateSpace=::World, Camera Query:NiagaraDataInterfaceCamera=DataInterface())
     Vector=<0 0 -500>
    )
    Force Strength=&MakeCustomFloatFromBool(Bool=$Particles.Player Close, True Float=30000, False Float=0)
    Module.Use Falloff=true
    Module.UseLinearForce=false
   )
   &LinearForce(
    Coordinate Space:ENiagaraCoordinateSpace=::World
    Force=&Multiply_VectorByFloat(
     Vector=&Subtract_Vector(
      A=Linked() {linkedVariable {name=Particles.Initial.Position, type=NiagaraPosition}}
      B=Linked() {linkedVariable {name=Particles.Position, type=NiagaraPosition}}
     )
     Float=&RandomRangeFloat(Minimum=0.25, Maximum=2.5)
    )
   )
   &Drag(Drag=4, Use Linear Drag=true, Use Rotational Drag=true)
   &CurlNoiseForce(
    Bias Noise=false
    Noise Frequency=5
    Noise Strength=&MakeCustomFloatFromBool(Bool=&Boolean_NotOperation(Bool=$Emitter.Enable Boids), True Float=1750, False Float=0)
    Pan Noise=true
    Pan Noise Field=<0 0 -1>
   )
   &SolveForcesAndVelocity(Acceleration Limit=350, Clamp Velocity=true, Limit Acceleration=true, Speed Limit=300)
   &self:Compute_Matrix(
    Normal:Vector3f=&Multiply_VectorByFloat(Vector=$Output.MoveToNearestDistanceFieldSurface_GPU001.VectorToNearestSurfaceNormal, Float=-1)
    Position:NiagaraPosition=$Particles.Position
    Velocity:Vector3f=$Particles.Velocity
   ) {module=Compute_Matrix}
   /SetVariables(Particles.MeshOrientation=&GetMatrixRotation(Matrix:NiagaraMatrix=$Output.Compute_Matrix.NiagaraMatrix)) {isSetParameters=true, module=SetVariables_#8}
   &InterpolateOverTime(Rate Of Change=4, Target Quat Value:Quat4f=$Particles.MeshOrientation, Variable Type=::NewEnumerator5)
   /SetVariables(Particles.MeshOrientation=&LerpQuaternion(
     Quaternion A:Quat4f=$Particles.Initial.MeshOrientation
     Quaternion B:Quat4f=$Particles.InterpolateOverTime.InterpolateOverTime_Quaternion.Moving Average
     Lerp Factor:NiagaraFloat=&RemapRange(
      Input Value To Remap=&VectorLength(Vector=$Particles.Presolve.Velocity)
      Input Min=0
      Input Max=10
      Output Min=0
      Output Max=1
     )
     Enable Spherical Linear Interpolation:NiagaraBool=true
    )) {isSetParameters=true, module=SetVariables_#9}
   /SetVariables(Particles.ParticleUpdateForce:Vector3f=$Transient.PhysicsForce) {isSetParameters=true, module=SetVariables_#10}
   &KillParticlesInVolume(
    Sphere Radius=80
    Volume Origin:Vector3f=&MakeVectorFromLinearColorRGB(Color=$NPC.FortniteNPC.PlayerPosition)
   )
   &RegionMask(
    Region Origin:Vector3f=&MakeVectorFromLinearColorRGB(Color=$NPC.FortniteNPC.PlayerPosition)
    Region Radius=90
    UseOffset=false
    UseRotation=false
   )
   &self:DeathSwitch_0_to_1(
    IsOutsideMask:NiagaraBool=&Boolean_NotOperation(Bool=$Particles.RegionMask.IsInsideMask)
    Position Array:NiagaraDataInterfaceArrayPosition=$System.Death Array
   ) {module=DeathSwitch_0_to_1}
   &KillParticles(Kill Particles=&self:ScratchDynamicInput_01())
  ]
  rendererClasses [&NiagaraMeshRendererProperties]
  renderers [{rendererClass=&NiagaraMeshRendererProperties, rendererIndex=0}]
 }
}
summary {
 systemName=NS_BugTunnel_GroundSwarm
 emitters [
  {bEnabled=false, emitterName=Bugs, simTarget=GPUComputeSim, rendererClasses [&NiagaraMeshRendererProperties]}
  {bEnabled=false, emitterName=Bugs001, simTarget=GPUComputeSim, rendererClasses [&NiagaraMeshRendererProperties]}
  {emitterName=Bugs002, simTarget=GPUComputeSim, rendererClasses [&NiagaraMeshRendererProperties]}
 ]
 userVariables [
  User.MeshScaleMult {type=NiagaraFloat, defaultValue=0.7, description=""}
  User.RadiusMult {type=NiagaraFloat, defaultValue=1, description=""}
  User.SpawnCount {type=NiagaraInt32, defaultValue=128}
 ]
}
